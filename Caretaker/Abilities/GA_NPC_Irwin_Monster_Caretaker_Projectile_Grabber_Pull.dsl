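; Gameplay ability for the Caretaker monster's grabber projectile that pulls a target toward its launcher, built on a knockback parent ability.
; On activation it plays the player-hit gameplay cue on the owner unless the avatar matches the vehicle tags, then defers to the parent.
; It overrides the target location to a point PullOffset from the launcher along the direction to the target, so the pull ends near the core.
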(event Ability|EventActivateAbility
  ;[1 1 1] Play Hit GC for the player.
  (bind _returnvalue (GameplayTags|HasAnyMatchingGameplayTags(Message) (Ability|GetAvatarActorfromActorInfo) (Variables|Default|GetTC_Vehicles)))
  (if (not _returnvalue)
    (Ability|ExecuteGameplayCueOnOwner (Variables|Default|GetGC_HitPlayer)))
  (|CallParentFunction))

(event Ability|EventFailedToActivatePassiveAbility
  (params (ActorInfo :type Gameplay%20Ability%20Actor%20Info%20Structure))
  (|CallParentFunction))

(fn GetTargetLocation
  ;[1 1 1] Have a little bit of an offset to the pull, but pull the actor pretty close to the core.
  (return (+ (Transformation|GetActorLocation (Variables|KnockbackParent|GetLauncher)) (* (Math|Vector|Normalize (- (Transformation|GetActorLocation (Variables|KnockbackParent|GetTarget)) (Transformation|GetActorLocation (Variables|KnockbackParent|GetLauncher)))) (Variables|Default|GetPullOffset)))))
